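; UI director for the RazorVault feature that adds or removes a progress bar scene as the player becomes or stops being a vault participant.
; On begin play and view target changes, it starts ability tasks watching the pawn's PlayerState for the participant tag.
; Tag removal starts a grace timer before clearing participation, and re-adding the tag cancels it.
; A delta gate updates the scene only when that state changes.
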(event EventBeginPlay
  (|CallParentFunction)
  (bind _asfort_player_controller (Utilities|Casting|CastToFortPlayerController (DynamicUIDirector|GetOwningLocalPlayerController)))
  (EventDispatchers|BindEventtoOnViewTargetChanged _asfort_player_controller (EventDispatchers|CreateEvent :!event OnViewTargetChanged))
  (CallFunction|ListenforParticipant (Game|Player|GetPlayerPawn _asfort_player_controller)))

(event EventEndPlay
  (params (EndPlayReason :type EEndPlayReason%20Enum :default "Destroyed"))
  (|CallParentFunction EndPlayReason)
  (bind _asfort_player_controller (Utilities|Casting|CastToFortPlayerController (DynamicUIDirector|GetOwningLocalPlayerController)))
  (EventDispatchers|UnbindEventfromOnViewTargetChanged _asfort_player_controller (EventDispatchers|CreateEvent :!event OnViewTargetChanged)))

(event Custom|ListenForParticipant
  (params (TargetActor :type Actor%20Object%20Reference))
  (bind _asfort_pawn (Utilities|Casting|CastToFortPawn TargetActor))
  (Ability|Tasks|WaitGameplayTagAddToActor (Variables|/Script/Engine.Pawn|GetPlayerState _asfort_pawn) (Variables|Vault|GetVaultParticipantTag)
    (:then
      (Ability|Tasks|WaitGameplayTagRemoveFromActor (Variables|/Script/Engine.Pawn|GetPlayerState _asfort_pawn) (Variables|Vault|GetVaultParticipantTag)
        (:Removed
          (if (Variables|Vault|GetIsPlayerAParticipant)
            (if (not (GameplayTags|HasMatchingGameplayTag (Ability|GetAbilitySystemComponent (Variables|/Script/Engine.Pawn|GetPlayerState _asfort_pawn)) (Variables|Vault|GetVaultParticipantTag)))
              (bind _returnvalue (Utilities|Time|SetTimerbyEvent (delegate Custom|OnGracePeriodEnded) (Variables|Default|GetWidgetGracePeriod)))
              (Variables|Default|SetGracePeriodTimerHandle _returnvalue))))))
    (:Added
      (Variables|Vault|SetIsPlayerAParticipant true)
      (CallFunction|RefreshProgressBarActiveState)
      (Utilities|Time|ClearandInvalidateTimerbyHandle (Variables|Default|GetGracePeriodTimerHandle)))))

(event Custom|OnGracePeriodEnded
  (Variables|Vault|SetIsPlayerAParticipant false)
  (CallFunction|RefreshProgressBarActiveState))

(fn OnViewTargetChanged
  (params (PlayerController :type Fort%20Player%20Controller%20Object%20Reference) (OldViewTarget :type Actor%20Object%20Reference) (NewViewTarget :type Actor%20Object%20Reference))
  (CallFunction|ListenforParticipant NewViewTarget))

(fn RefreshProgressBarActiveState
  (bind _progressbarscene (Variables|Vault|GetProgressBarScene))
  (bind _updated_variable (Utilities|FlowControl|DeltaGateWildcard (Variables|Vault|GetIsProgressBarActive) (Variables|Vault|GetIsPlayerAParticipant))
    (:Success
      (if _updated_variable
        (DynamicUIDirector|AddScene _progressbarscene)
        (else
          (DynamicUIDirector|RemoveScene _progressbarscene))))))

(fn ConstructionScript)
