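# UI material for a minimap vault marker that composites a vault icon with an optional boss icon, defaulting to a Baron texture.
# Each icon is a texture parameter sampled with clamped UI UVs, with its own center-based scale and X/Y offset parameters.
# Boss alpha, capped by ShowBossIcon, blends boss color over vault color, and its maximum with vault alpha forms the combined translucent shape.
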Material /VividRazorVaultGameplay/UI/Materials/M_UI_Minimap_Vault
graph {
 properties {bAutomaticallySetUsageInEditor=false, blendMode=BLEND_Translucent, materialDomain=MD_UI, usedShadingModels=Default Lit}
 nodes [
  n1 = MaterialFunctionCall() {materialFunction=&GetUserInterfaceUV}
  n2 = ScalarParameter() {defaultValue=0.7, parameterName=VaultIcon_Scale}
  n3 = AppendVector(A=@n2, B=@n2)
  n4 = ScalarParameter() {defaultValue=0.2, parameterName=VaultIcon_OffsetX}
  n5 = ScalarParameter() {parameterName=VaultIcon_OffsetY}
  n6 = TextureSampleParameter2D(UVs=Add(
    A=MaterialFunctionCall(UVs=@n1.Normalized UV, Texture Scale=@n3) {materialFunction=&ScaleUVsByCenter}.UVs
    B=AppendVector(A=@n4, B=@n5)
   )) {
   parameterName=VaultIcon
   samplerSource=SSM_Clamp_WorldGroupSettings
   texture=/VividRazorVaultGameplay/UI/Textures/T_UI_Minimap_Icon_Vaults_Idle
  }
  n7 = MaterialFunctionCall() {materialFunction=&GetUserInterfaceUV}
  n8 = ScalarParameter() {defaultValue=0.55, parameterName=BossIcon_Scale}
  n9 = AppendVector(A=@n8, B=@n8)
  n10 = ScalarParameter() {defaultValue=-0.45, parameterName=BossIcon_OffsetX}
  n11 = ScalarParameter() {parameterName=BossIcon_OffsetY}
  n12 = TextureSampleParameter2D(UVs=Add(
    A=MaterialFunctionCall(UVs=@n7.Normalized UV, Texture Scale=@n9) {materialFunction=&ScaleUVsByCenter}.UVs
    B=AppendVector(A=@n10, B=@n11)
   )) {parameterName=BossIcon, samplerSource=SSM_Clamp_WorldGroupSettings, texture=/VividRazorVaultGameplay/UI/Textures/T_UI_Minimap_Icon_Baron}
  n13 = Min(A=@n12.A, B=ScalarParameter() {defaultValue=1, parameterName=ShowBossIcon})
  LinearInterpolate(A=@n6.RGB, B=@n12.RGB, Alpha=@n13)
  Max(A=@n13, B=@n6.A)
 ]
}
